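# Sound cue for the rubble bucket container foley in the loot sound set, routed to the Fort_Loot_Collect sound class.
# A single wave player feeds the Rubble_Bucket wave straight to the output, with no randomization or modulation.
# Inline attenuation overrides take precedence over the assigned Fort_Prop_Search_Att asset to set the falloff.
# A concurrency override blocks new instances once the limit is reached instead of stopping playing ones.
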SoundCue /Game/Sounds/Foley_Loot/Containers/Rubble/Rubble_Bucket_Cue
root=WavePlayer_2
nodes [{id=WavePlayer_2, properties {soundWaveAssetPtr=&/Rubble_Bucket}}]
properties {
 attenuationSettings=/Game/Sounds/Attenuation/Fort_Prop_Search_Att
 bOverrideAttenuation=true
 bOverrideConcurrency=true
 maxDistance=1500
 soundClassObject=&Fort_Loot_Collect
 attenuationOverrides {falloffDistance=1000, attenuationShapeExtents {x=500}}
 concurrencyOverrides {resolutionRule=PreventNew}
}
